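# Sound cue for the destruction of the GuitarRustBucket_A prop.
# A random node picks between two destruction wave variants with equal weight.
# It uses a medium-object demolition attenuation, the impacts sound class and a four-max object-impact concurrency limit, with a boosted volume.
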SoundCue /Hera/Audio/Props/GuitarRustBucket_A/SC_GuitarRustBucket_A_Desctruction
root=Random_0
nodes [
 {id=Random_0, children [WavePlayer_2,WavePlayer_3], properties {weights [1,1]}}
 {id=WavePlayer_2, properties {soundWaveAssetPtr=&/GuitarRustBucket_A_Destruction_01}}
 {id=WavePlayer_3, properties {soundWaveAssetPtr=&/GuitarRustBucket_A_Destruction_02}}
]
properties {
 attenuationSettings=/Game/Sounds/Attenuation/Fort_Demolition_Destroy_Medium_Sized_Object_Att
 maxDistance=4000
 soundClassObject=&Fort_Impacts_PSM
 volumeMultiplier=5
 concurrencySet [&CONC_Object_Impacts_FourMax]
}
